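# Unlit alpha-composite material for the swirling DudeBro vortex in Figment, usable on meshes or mesh particles.
# Twisted UVs drive panned smoke and normal textures that feed density, Fresnel edge opacity and fake highlights, tinted by Cube lighting and time of day.
# Distance, depth and V-gradient fades shape opacity, with particle-driven erosion.
# Quality switches reserve rotation, world-aligned noise vertex displacement and animated-normal shading for High.
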Material /Figment_S06/Athena/Environments/DudeBro/M_Athena_DudeBro_Vortex_Figment
graph {
 properties {
  bEnableResponsiveAA=true
  bUsedWithMeshParticles=true
  bUsedWithNiagaraMeshParticles=true
  blendMode=BLEND_AlphaComposite
  pixelDepthOffsetMode=PDOM_Legacy
  shadingModel=MSM_Unlit
  translucencyLightingMode=TLM_VolumetricPerVertexDirectional
  usedShadingModels=Unlit
 }
 nodes [
  n1 = Add(
   A=StaticSwitchParameter(True=ObjectPositionWS(), False=ActorPositionWS()) {parameterName=UseObjectPosition?}
   B=VectorParameter() {group=CubeLighting, parameterName=CubeLightingActorOffset, defaultValue=<rgba 0 0 -1000 0>}.RGB
  )
  n2 = WorldPosition()
  n3 = ScalarParameter() {defaultValue=7500, group=CubeLighting, parameterName=CubeLightingRadius}
  n4 = ScalarParameter() {defaultValue=0.2, group=CubeLighting, parameterName=CubeLightingHardness}
  n5 = VertexInterpolator(VS=Multiply(
    A=SphereMask(A=@n1, B=@n2.XYZ, Radius=@n3, Hardness=@n4)
    B=VectorParameter() {group=CubeLighting, parameterName=CubeLightingColor, defaultValue=<rgba 7 1.3571 6.7302 1>}.RGB
   ))
  n6 = VectorParameter() {group=Base Color, parameterName=BaseColorLow, defaultValue=<rgba 0.6354 0.693 1 1>}
  n7 = VectorParameter() {group=Base Color, parameterName=BaseColorHigh, defaultValue=<rgba 1.4253 1.5549 2 1>}
  n8 = ScalarParameter() {defaultValue=0.45, group=FakeHighlight, parameterName=MobileColorScale}
  n9 = ScalarParameter() {defaultValue=0.1, group=FakeHighlight, parameterName=FakeHighlightScale-}
  n10 = ScalarParameter() {defaultValue=1.25, group=FakeHighlight, parameterName=FakeHighlightScale+}
  n11 = VertexNormalWS()
  n12 = Constant3Vector() {constant=<rgba 0 0 1 0>}
  n13 = StaticSwitchParameter(True=Normalize(VectorInput=Transform(None=@n12) {transformSourceType=TRANSFORMSOURCE_Local}), False=@n12) {parameterName=IsLobbySupportMeshRotation}
  n14 = Constant()
  n15 = VectorParameter() {group=Opacity, parameterName=FresnelNormalIntensity, defaultValue=<rgba 0.2 0.2 1 0>}
  n16 = MaterialFunctionCall(Float2=TextureCoordinate()) {materialFunction=&BreakOutFloat2Components}
  n17 = OneMinus(None=@n16.G)
  n18 = ScalarParameter() {defaultValue=2.7, group=UV Twist, parameterName=Twist Exponent}
  n19 = Add(A=@n16.R, B=Multiply(A=Power(Base=@n17, Exp=@n18), B=ScalarParameter() {defaultValue=0.17, group=UV Twist, parameterName=Twist Amount}))
  n20 = ScalarParameter() {defaultValue=0.145, group=UV Twist, parameterName=PanVBias}
  n21 = Power(
   Base=Multiply(A=Add(A=@n17, B=@n20), B=ScalarParameter() {defaultValue=3, group=UV Twist, parameterName=PanVScale})
   Exp=ScalarParameter() {defaultValue=4.7, group=UV Twist, parameterName=PanVExponent}
  )
  n22 = VertexInterpolator(VS=StaticSwitchParameter(True=AppendVector(A=@n19, B=@n21), False=TextureCoordinate()) {defaultValue=true, group=UV Twist, parameterName=UVTwist})
  n23 = ScalarParameter() {defaultValue=2, group=Density, parameterName=DensityTileU1}
  n24 = Multiply(
   A=ScalarParameter() {defaultValue=0.0017, group=Density, parameterName=DensityTileV1}
   B=DynamicParameter() {defaultValue=<rgba 0 1 1 1>, paramNames [Erode, WPOVertexNormalScale, PanScale, TileVScale]}.TileVScale
  )
  n25 = ScalarParameter() {defaultValue=0.03, group=Density, parameterName=DensityPanSpeedU1}
  n26 = ScalarParameter() {defaultValue=-0.107, group=Density, parameterName=DensityPanSpeedV1}
  n27 = Panner(
   Coordinate=Multiply(A=@n22.PS, B=AppendVector(A=@n23, B=@n24))
   Speed=Multiply(
    A=AppendVector(A=@n25, B=@n26)
    B=DynamicParameter() {defaultValue=<rgba 0 1 1 1>, paramNames [Erode, WPOVertexNormalScale, PanScale, TileVScale]}.PanScale
   )
  ) {bFractionalPart=true}
  n28 = ScalarParameter() {defaultValue=3, group=Density, parameterName=DensityTileU2}
  n29 = Multiply(
   A=ScalarParameter() {defaultValue=0.0113, group=Density, parameterName=DensityTileV2}
   B=DynamicParameter() {defaultValue=<rgba 0 1 1 1>, paramNames [Erode, WPOVertexNormalScale, PanScale, TileVScale]}.TileVScale
  )
  n30 = ScalarParameter() {group=Density, parameterName=DensityPanSpeedU2}
  n31 = ScalarParameter() {defaultValue=-0.1728, group=Density, parameterName=DensityPanSpeedV2}
  n32 = Panner(
   Coordinate=Multiply(A=@n22.PS, B=AppendVector(A=@n28, B=@n29))
   Speed=Multiply(
    A=AppendVector(A=@n30, B=@n31)
    B=DynamicParameter() {defaultValue=<rgba 0 1 1 1>, paramNames [Erode, WPOVertexNormalScale, PanScale, TileVScale]}.PanScale
   )
  ) {bFractionalPart=true}
  n33 = Add(
   A=TextureSample(UVs=@n27) {samplerType=SAMPLERTYPE_Normal, texture=/Game/Global/Textures/Default/Tileable_Patterns/2ndInfectedMapNormal}.RGB
   B=TextureSample(UVs=@n32) {samplerType=SAMPLERTYPE_Normal, texture=&2ndInfectedMapNormal}.RGB
  )
  n34 = AppendVector(
   A=ScalarParameter() {defaultValue=1, group=Density, parameterName=DensityTileU3}
   B=Multiply(
    A=ScalarParameter() {defaultValue=0.001, group=Density, parameterName=DensityTileV3}
    B=DynamicParameter() {defaultValue=<rgba 0 1 1 1>, paramNames [Erode, WPOVertexNormalScale, PanScale, TileVScale]}.TileVScale
   )
  )
  n35 = AppendVector(
   A=ScalarParameter() {defaultValue=0.01, group=Density, parameterName=DensityPanSpeedU3}
   B=ScalarParameter() {defaultValue=-0.0405, group=Density, parameterName=DensityPanSpeedV3}
  )
  n36 = DynamicParameter() {defaultValue=<rgba 0 1 1 1>, paramNames [Erode, WPOVertexNormalScale, PanScale, TileVScale]}
  n37 = TextureSampleParameter2D(UVs=Panner(Coordinate=Multiply(A=@n22.PS, B=@n34), Speed=Multiply(A=@n35, B=@n36.PanScale)) {bFractionalPart=true}) {parameterName=NormalDistortTexture, samplerType=SAMPLERTYPE_Normal, texture=/Game/Athena/Environments/DudeBro/T_Dudebro_Noise_N_01}
  n38 = Normalize(VectorInput=StaticSwitchParameter(True=@n33, False=Add(A=@n33, B=@n37.RGB)) {parameterName=InvertedMesh})
  n39 = StaticSwitchParameter(True=@n11, False=Transform(None=Multiply(A=@n38, B=@n15.RGB))) {parameterName=InvertedMesh}
  n40 = Multiply(A=Time(), B=ScalarParameter() {defaultValue=0.025, group=WPO, parameterName=RotateSpeed})
  n41 = QualitySwitch(
   Default=@n11
   Low=@n11
   High=Add(A=RotateAboutAxis(NormalizedRotationAxis=@n13, RotationAngle=@n40, PivotPoint=@n14, Position=@n39), B=@n39)
   Medium=@n11
  )
  n42 = DotProduct(
   A=Transform(None=@n41) {transformSourceType=TRANSFORMSOURCE_World, transformType=TRANSFORM_View}
   B=VectorParameter() {group=FakeHighlight, parameterName=FakeHighlightVector, defaultValue=<rgba 0.4 -0.1 0 0>}.RGB
  )
  n43 = ScalarParameter() {defaultValue=0.55, group=Opacity, parameterName=Fresnel Edge Center}
  n44 = MaterialFunctionCall(
   Brightness=ScalarParameter() {defaultValue=2, group=Opacity, parameterName=Fresnel Brightness}
   Exponent=ScalarParameter() {defaultValue=1.5, group=Opacity, parameterName=Fresnel Exponent}
   Vertex / Pixel Normal=@n41
  ) {materialFunction=&MF_Fresnel_FX}
  n45 = Saturate(None=@n44.Fresnel Double Sided)
  n46 = ScalarParameter() {defaultValue=0.45, desc=0.45, group=Opacity, parameterName=Fresnel Edge Radius}
  n47 = Constant()
  n48 = ScalarParameter() {defaultValue=0.3, group=Opacity, parameterName=FresnelEdgeOpacity_Top}
  n49 = LinearInterpolate(
   A=@n48
   B=ScalarParameter() {defaultValue=0.3, group=Opacity, parameterName=FresnelEdgeOpacity}
   Alpha=ComponentMask(None=TextureCoordinate()) {g=true}
  )
  n50 = QualitySwitch(Default=@n49, Low=@n48, High=@n48, Medium=@n48, Epic=@n49)
  n51 = ScalarParameter() {defaultValue=0.3, group=Opacity, parameterName=FresnelEdgeOpacity}
  n52 = StaticSwitchParameter(
   True=Multiply(
    A=SphereMask(A=@n43, B=@n45, Radius=@n46, Hardness=@n47)
    B=StaticSwitchParameter(True=@n50, False=@n51) {parameterName=UseOuterEdgeTopFade?}
   )
   False=Constant()
  ) {defaultValue=true, group=Opacity, parameterName=UseOuterEdge}
  n53 = Add(
   A=Power(Base=Saturate(None=@n42), Exp=ScalarParameter() {defaultValue=1.15, group=FakeHighlight, parameterName=FakeHighlightExponent})
   B=@n52
  )
  n54 = QualitySwitch(
   Default=@n8
   Low=@n8
   High=StaticSwitchParameter(True=@n8, False=LinearInterpolate(A=@n9, B=@n10, Alpha=@n53)) {parameterName=InvertedMesh}
   Medium=@n8
  )
  n55 = Add(
   A=TextureSample(UVs=@n27) {texture=/Game/Effects/Fort_Effects/Textures/Tile/T_CloudScape_08}.R
   B=TextureSample(UVs=@n32) {texture=&T_Inky_Smoke_Tile}.R
  )
  n56 = DotProduct(A=@n37.RGB, B=VectorParameter() {group=Density, parameterName=DOT, defaultValue=<rgba 0.6 0.6 0.4 1>}.RGB)
  n57 = Add(
   A=Multiply(A=@n55, B=ScalarParameter() {defaultValue=0.7, group=Density, parameterName=DensityScale})
   B=StaticSwitchParameter(True=Constant(), False=Multiply(A=@n55, B=@n56)) {parameterName=InvertedMesh}
  )
  n58 = Multiply(A=@n57, B=@n54)
  n59 = Multiply(
   A=LinearInterpolate(
    A=LinearInterpolate(A=@n6.RGB, B=@n7.RGB, Alpha=@n58)
    B=VectorParameter() {group=Base Color, parameterName=OuterEdgeColor, defaultValue=<rgba 0.4641 0.5537 0.8 1>}.RGB
    Alpha=@n52
   )
   B=@n54
  )
  n60 = StaticSwitchParameter(True=Add(A=Multiply(A=@n5.PS, B=@n59), B=@n59), False=@n59) {defaultValue=true, group=CubeLighting, parameterName=CubeLighting}
  n61 = VectorParameter() {group=TimeOfDay, parameterName="Color_Morning ", defaultValue=<rgba 1 0.5302 0.5601 0>}
  n62 = VectorParameter() {group=TimeOfDay, parameterName=Color_Day, defaultValue=<rgba 0.4635 0.6649 1 0>}
  n63 = VectorParameter() {group=TimeOfDay, parameterName=Color_Evening, defaultValue=<rgba 1 0.3438 0.3997 0>}
  n64 = VectorParameter() {group=TimeOfDay, parameterName=Color_Night, defaultValue=<rgba 0.4186 0.4321 0.8229 0>}
  n65 = StaticSwitchParameter(
   True=VertexInterpolator(VS=MaterialFunctionCall(Color_Morning=@n61.RGB, Color_Day=@n62.RGB, Color_Evening=@n63.RGB, Color_Night=@n64.RGB) {materialFunction=&Color_Time_of_Day}.Result).PS
   False=Constant() {r=1}
  ) {defaultValue=true, group=TimeOfDay, parameterName=UseTimeOfDayTint}
  n66 = Saturate(None=@n57)
  n67 = ScalarParameter() {defaultValue=0.75, group=Opacity, parameterName=OpacityDetailLerp}
  n68 = ScalarParameter() {defaultValue=0.325, group=Opacity, parameterName=Inner Opacity}
  n69 = Add(A=Multiply(A=LinearInterpolate(B=@n66, Alpha=@n67) {constA=1}, B=Multiply(A=@n44.Fresnel Double Sided, B=@n68)), B=@n52)
  n70 = MaterialFunctionCall(
   Opacity In=@n69
   Position=WorldPosition().XYZ
   Fade_Start_Distance=ScalarParameter() {defaultValue=1500, group=Opacity, parameterName=Fade_Start_Distance}
   Fade_End_Distance=ScalarParameter() {defaultValue=200, group=Opacity, parameterName=Fade_End_Distance}
  ) {materialFunction=&MF_Distance_SyncedFade_FX}
  n71 = Saturate(None=QualitySwitch(Default=@n69, Low=@n69, High=@n70.Result, Medium=@n70.Result, Epic=@n70.Result))
  n72 = StaticSwitchParameter(True=DepthFade(Opacity=@n71, FadeDistance=ScalarParameter() {defaultValue=1000, group=Opacity, parameterName=DepthFadeDistance}), False=@n71) {defaultValue=true, group=Opacity, parameterName=UseDepthFade}
  n73 = ComponentMask(None=TextureCoordinate()) {g=true}
  n74 = ScalarParameter() {defaultValue=0.5, group=Opacity, parameterName=FadeCenter}
  n75 = ScalarParameter() {defaultValue=0.48, group=Opacity, parameterName=FadeRadius}
  n76 = ScalarParameter() {defaultValue=0.75, group=Opacity, parameterName=FadeHardness}
  n77 = ComponentMask(None=VertexInterpolator(VS=SphereMask(A=@n73, B=@n74, Radius=@n75, Hardness=@n76)).PS) {desc=Possible bug with VertexInterpolator, r=true}
  n78 = Multiply(A=Subtract(A=QualitySwitch(Default=@n71, Low=@n71, High=@n72, Medium=@n71, Epic=@n72), B=OneMinus(None=@n77)), B=@n77) {constB=1.5}
  n79 = DynamicParameter() {defaultValue=<rgba 0 1 1 1>, paramNames [Erode, WPOVertexNormalScale, PanScale, TileVScale]}
  n80 = StaticSwitchParameter(
   True=Multiply(A=Subtract(A=@n78, B=@n79.Erode), B=ParticleColor().A)
   False=Multiply(A=ScalarParameter() {defaultValue=1, group=Opacity, parameterName=OverallOpacity}, B=@n78)
  ) {parameterName=IsParticle}
  Multiply(A=Multiply(A=ParticleColor().RGB, B=Multiply(A=@n60, B=@n65)), B=@n80) {desc=Alpha Composite}
  n81 = Constant()
  n82 = Constant3Vector() {constant=<rgba 0 0 1 0>}
  n83 = StaticSwitchParameter(True=Normalize(VectorInput=Transform(None=@n82) {transformSourceType=TRANSFORMSOURCE_Local}), False=@n82) {parameterName=IsLobbySupportMeshRotation}
  n84 = ObjectPositionWS()
  n85 = WorldPosition()
  n86 = DynamicParameter() {defaultValue=<rgba 0 1 1 1>, paramNames [Erode, WPOVertexNormalScale, PanScale, TileVScale]}
  n87 = VectorParameter() {group=WPO, parameterName=WPOVertexNormalScale}
  n88 = TextureObjectParameter() {group=WPO, parameterName=WPONoise, samplerType=SAMPLERTYPE_Normal, texture=/Game/Global/Textures/Default/Noise/T_SwirlNoise_N}
  n89 = WorldPosition()
  n90 = Multiply(A=Time(), B=VectorParameter() {group=WPO, parameterName=WPOPanSpeed, defaultValue=<rgba 0 0 -500 0>}.RGB)
  n91 = MaterialFunctionCall(Float3=MaterialFunctionCall(
    TextureObject=@n88
    TextureSize=VectorParameter() {group=WPO, parameterName=WPOWorldScale, defaultValue=<rgba 25000 25000 25000 0>}.RGB
    WorldPosition=Add(A=@n89.XYZ, B=@n90)
   ) {materialFunction=&WorldAlignedTexture}.XYZ Texture) {materialFunction=&BreakOutFloat3Components}
  n92 = MaterialFunctionCall(X=@n91.R, Y=@n91.R, Z=Subtract(A=@n91.B)) {materialFunction=&MakeFloat3}
  n93 = Multiply(A=@n92.Result) {constB=-1}
  n94 = Add(
   A=Multiply(A=StaticSwitchParameter(True=@n86.WPOVertexNormalScale, False=@n87.RGB) {parameterName=IsParticle}, B=VertexNormalWS())
   B=Multiply(
    A=StaticSwitchParameter(True=@n93, False=@n92.Result) {group=WPO, parameterName=InvertedWPO}
    B=VectorParameter() {group=WPO, parameterName=WPO Scale, defaultValue=<rgba 500 500 0 0>}.RGB
   )
  )
  n95 = TextureCoordinate()
  n96 = Power(
   Base=OneMinus(None=MaterialFunctionCall(Float2=@n95) {materialFunction=&BreakOutFloat2Components}.G)
   Exp=ScalarParameter() {defaultValue=3.5, group=WPO, parameterName=WPOMaskExponent}
  )
  QualitySwitch(
   Default=@n81
   Low=@n81
   High=Add(A=RotateAboutAxis(NormalizedRotationAxis=@n83, RotationAngle=@n40, PivotPoint=@n84, Position=@n85.XYZ), B=Multiply(A=@n94, B=@n96))
   Medium=@n81
  )
  Reroute(None=@n80)
 ]
}
